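# Animation Blueprint for the Durrburger Hero backpack cosmetic that adds physics-driven secondary motion to its bones.
# The input pose runs in component space through eleven chained Anim Dynamics nodes: ten drv_ bones, then dyn_fryMound.
# The drv_ bones vary spring stiffness, Z travel and blend alpha for uneven jiggle, while the fry mound blends fully with unscaled component motion.
# The event graph holds only unconnected stub nodes.
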AnimBlueprint /BRCosmetics/Accessories/FORT_Backpacks/M_MED_Durrburger_Hero_Backpack/Meshes/Backpack_Durrburger_Hero_Skeleton_AnimBP
target_skeleton=/Game/Accessories/FORT_Backpacks/M_MED_Durrburger_Hero_Backpack/Meshes/Backpack_Durrburger_Hero_Skeleton
graphs [
 AnimGraph {nodes [
   AnimGraphNode_SubInput_0 = Input Pose()
   n1 = Anim Dynamics - Bone: drv_mstr_01(ComponentPose=Local To Component(LocalPose=@AnimGraphNode_SubInput_0), Alpha=0.25, ExternalForce="0,0,0") {
    alpha=0.9
    angularDampingOverride=0.98
    angularSpringConstant=400
    bAngularSpring=true
    bGravityOverrideInSimSpace=true
    bOverrideAngularDamping=true
    numSolverIterationsPostUpdate=4
    numSolverIterationsPreUpdate=16
    alphaBoolBlend {blendOption=Linear}
    boundBone {boneName=drv_mstr_01}
    componentAppliedLinearAccClamp=<1000 1000 1000>
    componentLinearAccScale=<0.1 1 1>
    componentLinearVelScale=<1 0.1 1>
    physicsBodyDefinitions [{
      sphereCollisionRadius=0
      boundBone {boneName=drv_mstr_01}
      boxExtents=<15 5 5>
      constraintSetup {
       angularTargetAxis=AxisZ
       twistAxis=AxisY
       angularLimitsMax=<20 20 20>
       angularLimitsMin=<-20 -20 -20>
       angularTarget=<0 0 1>
       linearAxesMax=<0 0 2>
      }
      localJointOffset=<0 0 -1>
     }]
   }
   n2 = Anim Dynamics - Bone: drv_mstr_02(ComponentPose=@n1, Alpha=0.5, ExternalForce="0,0,0") {
    alpha=0.9
    angularDampingOverride=0.98
    angularSpringConstant=600
    bAngularSpring=true
    bGravityOverrideInSimSpace=true
    bOverrideAngularDamping=true
    numSolverIterationsPostUpdate=4
    numSolverIterationsPreUpdate=16
    alphaBoolBlend {blendOption=Linear}
    boundBone {boneName=drv_mstr_02}
    componentAppliedLinearAccClamp=<1000 1000 1000>
    componentLinearAccScale=<0.1 1 1>
    componentLinearVelScale=<1 0.1 1>
    physicsBodyDefinitions [{
      sphereCollisionRadius=0
      boundBone {boneName=drv_mstr_02}
      boxExtents=<15 5 5>
      constraintSetup {
       angularTargetAxis=AxisZ
       twistAxis=AxisY
       angularLimitsMax=<20 20 20>
       angularLimitsMin=<-20 -20 -20>
       angularTarget=<0 0 1>
       linearAxesMax=<0 0 1>
      }
      localJointOffset=<0 0 -1>
     }]
   }
   n3 = Anim Dynamics - Bone: drv_mstr_03(ComponentPose=@n2, Alpha=0.175, ExternalForce="0,0,0") {
    alpha=0.9
    angularDampingOverride=0.98
    angularSpringConstant=750
    bAngularSpring=true
    bGravityOverrideInSimSpace=true
    bOverrideAngularDamping=true
    numSolverIterationsPostUpdate=4
    numSolverIterationsPreUpdate=16
    alphaBoolBlend {blendOption=Linear}
    boundBone {boneName=drv_mstr_03}
    componentAppliedLinearAccClamp=<1000 1000 1000>
    componentLinearAccScale=<0.1 1 1>
    componentLinearVelScale=<1 0.1 1>
    physicsBodyDefinitions [{
      sphereCollisionRadius=0
      boundBone {boneName=drv_mstr_03}
      boxExtents=<15 5 5>
      constraintSetup {
       angularTargetAxis=AxisZ
       twistAxis=AxisY
       angularLimitsMax=<20 20 20>
       angularLimitsMin=<-20 -20 -20>
       angularTarget=<0 0 1>
       linearAxesMax=<0 0 5>
      }
      localJointOffset=<0 0 -1>
     }]
   }
   n4 = Anim Dynamics - Bone: drv_mstr_04(ComponentPose=@n3, Alpha=0.125, ExternalForce="0,0,0") {
    alpha=0.9
    angularDampingOverride=0.98
    angularSpringConstant=650
    bAngularSpring=true
    bGravityOverrideInSimSpace=true
    bOverrideAngularDamping=true
    numSolverIterationsPostUpdate=4
    numSolverIterationsPreUpdate=16
    alphaBoolBlend {blendOption=Linear}
    boundBone {boneName=drv_mstr_04}
    componentAppliedLinearAccClamp=<1000 1000 1000>
    componentLinearAccScale=<0.1 1 1>
    componentLinearVelScale=<1 0.1 1>
    physicsBodyDefinitions [{
      sphereCollisionRadius=0
      boundBone {boneName=drv_mstr_04}
      boxExtents=<15 5 5>
      constraintSetup {
       angularTargetAxis=AxisZ
       twistAxis=AxisY
       angularLimitsMax=<20 20 20>
       angularLimitsMin=<-20 -20 -20>
       angularTarget=<0 0 1>
       linearAxesMax=<0 0 3>
      }
      localJointOffset=<0 0 -1>
     }]
   }
   n5 = Anim Dynamics - Bone: drv_mstr_05(ComponentPose=@n4, Alpha=0.35, ExternalForce="0,0,0") {
    alpha=0.9
    angularDampingOverride=0.98
    angularSpringConstant=450
    bAngularSpring=true
    bGravityOverrideInSimSpace=true
    bOverrideAngularDamping=true
    numSolverIterationsPostUpdate=4
    numSolverIterationsPreUpdate=16
    alphaBoolBlend {blendOption=Linear}
    boundBone {boneName=drv_mstr_05}
    componentAppliedLinearAccClamp=<1000 1000 1000>
    componentLinearAccScale=<0.1 1 1>
    componentLinearVelScale=<1 0.1 1>
    physicsBodyDefinitions [{
      sphereCollisionRadius=0
      boundBone {boneName=drv_mstr_05}
      boxExtents=<15 5 5>
      constraintSetup {
       angularTargetAxis=AxisZ
       twistAxis=AxisY
       angularLimitsMax=<20 20 20>
       angularLimitsMin=<-20 -20 -20>
       angularTarget=<0 0 1>
       linearAxesMax=<0 0 2>
      }
      localJointOffset=<0 0 -1>
     }]
   }
   n6 = Anim Dynamics - Bone: drv_mstr_06(ComponentPose=@n5, Alpha=0.08, ExternalForce="0,0,0") {
    alpha=0.9
    angularDampingOverride=0.98
    angularSpringConstant=550
    bAngularSpring=true
    bGravityOverrideInSimSpace=true
    bOverrideAngularDamping=true
    numSolverIterationsPostUpdate=4
    numSolverIterationsPreUpdate=16
    alphaBoolBlend {blendOption=Linear}
    boundBone {boneName=drv_mstr_06}
    componentAppliedLinearAccClamp=<1000 1000 1000>
    componentLinearAccScale=<0.1 1 1>
    componentLinearVelScale=<1 0.1 1>
    physicsBodyDefinitions [{
      sphereCollisionRadius=0
      boundBone {boneName=drv_mstr_06}
      boxExtents=<15 5 5>
      constraintSetup {
       angularTargetAxis=AxisZ
       twistAxis=AxisY
       angularLimitsMax=<20 20 20>
       angularLimitsMin=<-20 -20 -20>
       angularTarget=<0 0 1>
       linearAxesMax=<0 0 6>
      }
      localJointOffset=<0 0 -1>
     }]
   }
   n7 = Anim Dynamics - Bone: drv_mstr_07(ComponentPose=@n6, Alpha=0.455, ExternalForce="0,0,0") {
    alpha=0.9
    angularDampingOverride=0.98
    angularSpringConstant=325
    bAngularSpring=true
    bGravityOverrideInSimSpace=true
    bOverrideAngularDamping=true
    numSolverIterationsPostUpdate=4
    numSolverIterationsPreUpdate=16
    alphaBoolBlend {blendOption=Linear}
    boundBone {boneName=drv_mstr_07}
    componentAppliedLinearAccClamp=<1000 1000 1000>
    componentLinearAccScale=<0.1 1 1>
    componentLinearVelScale=<1 0.1 1>
    physicsBodyDefinitions [{
      sphereCollisionRadius=0
      boundBone {boneName=drv_mstr_07}
      boxExtents=<15 5 5>
      constraintSetup {
       angularTargetAxis=AxisZ
       twistAxis=AxisY
       angularLimitsMax=<20 20 20>
       angularLimitsMin=<-20 -20 -20>
       angularTarget=<0 0 1>
       linearAxesMax=<0 0 4>
      }
      localJointOffset=<0 0 -1>
     }]
   }
   n8 = Anim Dynamics - Bone: drv_mstr_08(ComponentPose=@n7, Alpha=0.125, ExternalForce="0,0,0") {
    alpha=0.9
    angularDampingOverride=0.98
    angularSpringConstant=450
    bAngularSpring=true
    bGravityOverrideInSimSpace=true
    bOverrideAngularDamping=true
    numSolverIterationsPostUpdate=4
    numSolverIterationsPreUpdate=16
    alphaBoolBlend {blendOption=Linear}
    boundBone {boneName=drv_mstr_08}
    componentAppliedLinearAccClamp=<1000 1000 1000>
    componentLinearAccScale=<0.1 1 1>
    componentLinearVelScale=<1 0.1 1>
    physicsBodyDefinitions [{
      sphereCollisionRadius=0
      boundBone {boneName=drv_mstr_08}
      boxExtents=<15 5 5>
      constraintSetup {
       angularTargetAxis=AxisZ
       twistAxis=AxisY
       angularLimitsMax=<20 20 20>
       angularLimitsMin=<-20 -20 -20>
       angularTarget=<0 0 1>
       linearAxesMax=<0 0 1>
      }
      localJointOffset=<0 0 -1>
     }]
   }
   n9 = Anim Dynamics - Bone: drv_child_09(ComponentPose=@n8, Alpha=0.2, ExternalForce="0,0,0") {
    alpha=0.9
    angularDampingOverride=0.98
    angularSpringConstant=350
    bAngularSpring=true
    bGravityOverrideInSimSpace=true
    bOverrideAngularDamping=true
    numSolverIterationsPostUpdate=4
    numSolverIterationsPreUpdate=16
    alphaBoolBlend {blendOption=Linear}
    boundBone {boneName=drv_child_09}
    componentAppliedLinearAccClamp=<1000 1000 1000>
    componentLinearAccScale=<0.1 1 1>
    componentLinearVelScale=<1 0.1 1>
    physicsBodyDefinitions [{
      sphereCollisionRadius=0
      boundBone {boneName=drv_child_09}
      boxExtents=<15 5 5>
      constraintSetup {
       angularTargetAxis=AxisZ
       twistAxis=AxisY
       angularLimitsMax=<20 20 20>
       angularLimitsMin=<-20 -20 -20>
       angularTarget=<0 0 1>
       linearAxesMax=<0 0 5>
      }
      localJointOffset=<0 0 -1>
     }]
   }
   n10 = Anim Dynamics - Bone: drv_mstr_010(ComponentPose=@n9, Alpha=0.25, ExternalForce="0,0,0") {
    alpha=0.9
    angularDampingOverride=0.98
    angularSpringConstant=500
    bAngularSpring=true
    bGravityOverrideInSimSpace=true
    bOverrideAngularDamping=true
    numSolverIterationsPostUpdate=4
    numSolverIterationsPreUpdate=16
    alphaBoolBlend {blendOption=Linear}
    boundBone {boneName=drv_mstr_010}
    componentAppliedLinearAccClamp=<1000 1000 1000>
    componentLinearAccScale=<0.1 1 1>
    componentLinearVelScale=<1 0.1 1>
    physicsBodyDefinitions [{
      sphereCollisionRadius=0
      boundBone {boneName=drv_mstr_010}
      boxExtents=<15 5 5>
      constraintSetup {
       angularTargetAxis=AxisZ
       twistAxis=AxisY
       angularLimitsMax=<20 20 20>
       angularLimitsMin=<-20 -20 -20>
       angularTarget=<0 0 1>
       linearAxesMax=<0 0 2>
      }
      localJointOffset=<0 0 -1>
     }]
   }
   n11 = Anim Dynamics - Bone: dyn_fryMound(ComponentPose=@n10, Alpha=1, ExternalForce="0,0,0") {
    alpha=0.9
    angularDampingOverride=0.98
    angularSpringConstant=450
    bAngularSpring=true
    bGravityOverrideInSimSpace=true
    bOverrideAngularDamping=true
    numSolverIterationsPostUpdate=4
    numSolverIterationsPreUpdate=16
    alphaBoolBlend {blendOption=Linear}
    boundBone {boneName=dyn_fryMound}
    componentAppliedLinearAccClamp=<1000 1000 1000>
    componentLinearAccScale=<1 1 1>
    componentLinearVelScale=<1 1 1>
    physicsBodyDefinitions [{
      sphereCollisionRadius=0
      boundBone {boneName=dyn_fryMound}
      boxExtents=<15 5 5>
      constraintSetup {angularTargetAxis=AxisZ, twistAxis=AxisY, angularLimitsMax=<20 20 20>, angularLimitsMin=<-20 -20 -20>, angularTarget=<0 0 1>}
      localJointOffset=<0 0 -0.5>
     }]
   }
   Output Pose(Result=Component To Local(ComponentPose=@n11)) {layerGroup=None}
  ]}
 EventGraph {nodes [
   TryGetPawnOwner()
   Event BlueprintUpdateAnimation()
  ]}
]
variables []
